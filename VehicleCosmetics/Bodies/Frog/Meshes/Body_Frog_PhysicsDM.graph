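# Physics asset for the Frog vehicle body cosmetic, providing its simplified collision shape.
# A single box on the root bone, raised and shifted slightly along X from the bone origin, stands in for the whole body.
# It has no other bodies and no constraints, so it defines no articulated parts.
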PhysicsAsset /VehicleCosmetics/Bodies/Frog/Meshes/Body_Frog_PhysicsDM
bodies {root [{shapeType=Box, extentX=444, extentY=216, extentZ=132, shapeName=root_box, center=<23.1128 0 97.1149>}]}
constraints []
